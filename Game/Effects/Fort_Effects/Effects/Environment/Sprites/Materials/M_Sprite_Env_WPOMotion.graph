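# Translucent unlit material for instanced or Niagara mesh environment sprites with glowing, blinking eyes, animated through world position offset.
# Custom bases rotate each mesh and its normals from camera-facing toward a LookAt point as camera distance grows.
# Offsets combine per-instance sine bobbing, ground height and curl-noise warble, plus a switch to a sunken hide position, apparently driven by camera proximity.
# Opacity combines a view-angle sphere mask with per-instance near-camera hiding.
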Material /Game/Effects/Fort_Effects/Effects/Environment/Sprites/Materials/M_Sprite_Env_WPOMotion
graph {
 properties {
  assetImportData="<self>:InterchangeAssetImportData_0"
  bUsedWithInstancedStaticMeshes=true
  bUsedWithNiagaraMeshParticles=true
  blendMode=BLEND_Translucent
  shadingModel=MSM_Unlit
  usedShadingModels=Unlit
 }
 nodes [
  If("A > B"=Constant() {r=1}, "A == B"=Constant(), "A < B"=Constant()) {constB=0.5}
  LocalPosition() {localOrigin=InstancePreSkinning}
  MaterialFunctionCall!(WorldPosition, PivotLocation, WS Normals, Custom Object Basis 1, Custom Object Basis 2, Custom Object Basis 3) {materialFunction=/Engine/Functions/Engine_MaterialFunctions02/WorldPositionOffset/AlignMeshToTheCamera}
  n1 = TransformPosition(None=WorldPosition().XYZ) {transformSourceType=TRANSFORMPOSSOURCE_World, transformType=TRANSFORMPOSSOURCE_Instance}
  n2 = Normalize(VectorInput=Subtract(A=CameraPositionWS(), B=ObjectPositionWS()))
  n3 = Normalize(VectorInput=CrossProduct(A=@n2, B=Constant3Vector() {constant=<rgba 0 0 1 0>}))
  n4 = Normalize(VectorInput=CrossProduct(A=@n3, B=@n2))
  n5 = MaterialFunctionCall(VectorToTransform=@n1, BasisX=@n2, BasisY=Multiply(A=@n3) {constB=-1}, BasisZ=@n4) {materialFunction=&Transform3x3Matrix}
  n6 = TransformPosition(None=WorldPosition().XYZ) {transformSourceType=TRANSFORMPOSSOURCE_World, transformType=TRANSFORMPOSSOURCE_Instance}
  n7 = Normalize(VectorInput=Subtract(A=VectorParameter() {parameterName=LookAt, defaultValue=<rgba 14131 70226 4306 0>}.RGB, B=ObjectPositionWS()))
  n8 = Normalize(VectorInput=CrossProduct(A=@n7, B=Constant3Vector() {constant=<rgba 0 0 1 0>}))
  n9 = Normalize(VectorInput=CrossProduct(A=@n8, B=@n7))
  n10 = MaterialFunctionCall(VectorToTransform=@n6, BasisX=@n7, BasisY=Multiply(A=@n8) {constB=-1}, BasisZ=@n9) {materialFunction=&Transform3x3Matrix}
  n11 = CameraPositionWS()
  n12 = WorldPosition()
  n13 = Power(
   Base=MaterialFunctionCall(Input=Distance(A=@n11, B=@n12.XYZ), Input Low=ScalarParameter(), Input High=ScalarParameter() {defaultValue=1500, parameterName=Param_1}) {materialFunction=&RemapValue_0-1}.Result
   Exp=ScalarParameter() {defaultValue=24, parameterName=Param_3}
  )
  n14 = LinearInterpolate(
   A=Multiply(
    A=Subtract(A=@n5.Transformed Vector, B=@n1)
    B=MaterialFunctionCall() {materialFunction=/Engine/Functions/Engine_MaterialFunctions02/WorldPositionOffset/ObjectScale}.Scale XYZ
   )
   B=Multiply(
    A=Subtract(A=@n10.Transformed Vector, B=@n6)
    B=MaterialFunctionCall() {materialFunction=&ObjectScale}.Scale XYZ
   )
   Alpha=@n13
  )
  MaterialFunctionCall!(Vector3=@n14, MaximumNumberOfDigits, UVs, DebugTextLocation RG_UpperRight BA_LowerLeft, Component Spacing) {materialFunction=/Engine/Functions/Engine_MaterialFunctions02/Utility/DebugFloat3Values}
  MaterialFunctionCall!(Use World Position Offsets) {materialFunction=/Engine/Functions/Engine_MaterialFunctions02/Utility/CameraVectorWithWPOOptions}
  MaterialFunctionCall!(Vector1, Vector2, Vector3) {materialFunction=/Engine/Functions/Engine_MaterialFunctions02/Math/MakeVectorsOrthogonal}
  MaterialFunctionCall() {materialFunction=&BoundingBoxBased_0-1_UVW}
  MaterialFunctionCall() {materialFunction=&ObjectPivotPoint}
  n15 = VectorParameter() {parameterName=Color, defaultValue=<rgba 1.25 0.7019 0.3613 1>}
  n16 = Multiply(A=@n15.RGB, B=ScalarParameter() {defaultValue=10, parameterName=Eye Brightness})
  n17 = TextureSampleParameter2D() {parameterName=EmissiveColorMap, texture=/Game/Effects/Fort_Effects/Effects/Environment/Sprites/Materials/Textures/T_EnvSprite_EyeGlows}
  n18 = Sine(None=Add(A=Time(), B=PerInstanceRandom()))
  n19 = LinearInterpolate(
   A=@n17.G
   B=TextureSample() {texture=/Game/Effects/Fort_Effects/Effects/Environment/Sprites/Materials/Textures/T_EnvSprite_EyeGlows_Blink}.G
   Alpha=Round(None=Saturate(None=@n18))
  )
  MaterialFunctionCall(Emissive=Add(A=@n15.RGB, B=Multiply(A=@n16, B=@n19))) {materialFunction=&MF_FX_EmissiveAdjustments}
  n20 = ScalarParameter() {defaultValue=0.07, parameterName=Overall Opacity}
  n21 = CameraVectorWS()
  n22 = ConstantBiasScale(None=MaterialFunctionCall() {materialFunction=/Game/Effects/Niagara/MasterMaterials/Functions/InstanceBounds_0-1_UVW}.RGB) {bias=-0.5, scale=2}
  n23 = SphereMask(
   A=Saturate(None=DotProduct(A=@n21, B=@n22))
   B=ScalarParameter() {defaultValue=0.25, parameterName=Mask Start}
   Radius=ScalarParameter() {defaultValue=0.7952, parameterName=Mask Radius}
   Hardness=ScalarParameter() {parameterName=Mask Hardness}
  )
  n24 = ScalarParameter() {defaultValue=500, parameterName=Hide Distance Near}
  n25 = ScalarParameter() {defaultValue=750, parameterName=Hide Distance Far}
  n26 = PerInstanceRandom()
  n27 = OneMinus(None=SphereMask(A=WorldPosition().XYZ, B=CameraPositionWS(), Radius=LinearInterpolate(A=@n24, B=@n25, Alpha=@n26)) {attenuationRadius=5500, hardnessPercent=0})
  Multiply(
   A=Multiply(A=Multiply(A=@n20, B=@n23), B=NamedRerouteUsage())
   B=Power(Base=@n27, Exp=ScalarParameter() {defaultValue=16, parameterName=Hide Threshold Power})
  )
  Multiply()
  Multiply(B=MaterialFunctionCall() {materialFunction=/Engine/Functions/Engine_MaterialFunctions02/WorldPositionOffset/V2/ObjectScale}.Scale XYZ)
  Multiply()
  n28 = TransformPosition(None=ObjectPositionWS()) {transformSourceType=TRANSFORMPOSSOURCE_World, transformType=TRANSFORMPOSSOURCE_Instance}
  n29 = Add(A=VectorParameter() {parameterName=Magnitue Low, defaultValue=<rgba 0 0 -1 0>}.RGB, B=@n28)
  n30 = Add(A=@n28, B=VectorParameter() {parameterName=Magnitude High, defaultValue=<rgba 0 0 10 0>}.RGB)
  n31 = Time()
  n32 = PerInstanceRandom()
  n33 = Sine(None=Multiply(A=Add(A=@n31, B=@n32), B=ScalarParameter() {defaultValue=0.25, parameterName=Param_2}))
  n34 = VectorParameter() {parameterName=Hide Place, defaultValue=<rgba 0 0 -500 0>}
  n35 = TransformPosition(None=ObjectPositionWS()) {transformSourceType=TRANSFORMPOSSOURCE_World, transformType=TRANSFORMPOSSOURCE_Instance}
  n36 = VectorParameter() {parameterName=Rest Position, defaultValue=<rgba 0 0 1 0>}
  n37 = NamedRerouteUsage()
  n38 = Constant() {r=1}
  n39 = Constant()
  n40 = Constant()
  n41 = Add(
   A=Add(A=NamedRerouteUsage(), B=LinearInterpolate(A=@n29, B=@n30, Alpha=@n33))
   B=LinearInterpolate(A=Add(A=@n34.RGB, B=@n35), B=Add(A=@n35, B=@n36.RGB), Alpha=If(A=@n37, "A > B"=@n38, "A == B"=@n39, "A < B"=@n40) {constB=0.05})
  )
  n42 = Add(
   A=TransformPosition(None=ObjectPositionWS()) {transformSourceType=TRANSFORMPOSSOURCE_World, transformType=TRANSFORMPOSSOURCE_Instance}
   B=LinearInterpolate(
    A=VectorParameter() {parameterName=Ground Offset Low, defaultValue=<rgba 0 0 22 0>}.RGB
    B=VectorParameter() {parameterName=Ground Offset High, defaultValue=<rgba 0 0 45 0>}.RGB
    Alpha=PerInstanceRandom()
   )
  )
  n43 = TransformPosition(None=WorldPosition().XYZ) {transformSourceType=TRANSFORMPOSSOURCE_World, transformType=TRANSFORMPOSSOURCE_Instance}
  n44 = ScalarParameter() {defaultValue=0.001, parameterName=Warble Scale}
  n45 = Time()
  n46 = VectorParameter() {defaultValue=<rgba 0 0 0.05 0>}
  n47 = TextureSample(UVs=Add(A=Multiply(A=@n43, B=@n44), B=Multiply(A=@n45, B=@n46.RGB))) {samplerType=SAMPLERTYPE_LinearColor, texture=/Game/Effects/FogVolume/VT_CurlNoise_LOW}
  n48 = ScalarParameter() {defaultValue=0.4, parameterName=Warble Amount}
  NamedRerouteDeclaration(None=Add(A=Add(A=@n41, B=@n42), B=Multiply(A=@n47.RGB, B=@n48))) {name=WPO, nodeColor=<rgba 0 0.6016 1 1>}
  n49 = VertexNormalWS()
  NamedRerouteDeclaration(None=LinearInterpolate(
    A=MaterialFunctionCall(VectorToTransform=@n49, BasisX=@n2, BasisY=@n3, BasisZ=@n4) {materialFunction=&Transform3x3Matrix}.Transformed Vector
    B=MaterialFunctionCall(VectorToTransform=@n49, BasisX=@n7, BasisY=@n8, BasisZ=@n9) {materialFunction=&Transform3x3Matrix}.Transformed Vector
    Alpha=@n13
   )) {name=Normals, nodeColor=<rgba 0 0.8594 1 1>}
  NamedRerouteDeclaration(None=@n27) {name=Distance, nodeColor=<rgba 0.6406 1 0 1>}
  NamedRerouteDeclaration(None=SphereMask(
    A=WorldPosition().XYZ
    B=CameraPositionWS()
    Radius=ScalarParameter() {defaultValue=6500, parameterName=View Distance Fade}
    Hardness=ScalarParameter() {parameterName=Param_6}
   )) {name=View Distance Fade, nodeColor=<rgba 0.6641 0 1 1>}
  NamedRerouteDeclaration(None=Divide(
    A=Transform(None=@n14) {transformSourceType=TRANSFORMSOURCE_Local}
    B=MaterialFunctionCall() {materialFunction=&ObjectScale}.Scale XYZ
   )) {name=Rotated Direction, nodeColor=<rgba 1 0 0.6797 1>}
  NamedRerouteUsage()
  NamedRerouteUsage()
  NamedRerouteUsage()
  ParticleColor()
  Power()
  ScalarParameter() {parameterName=Blend}
  Subtract(
   A=@n10.Transformed Vector
   B=TransformPosition(None=Constant3Vector()) {transformSourceType=TRANSFORMPOSSOURCE_Instance, transformType=TRANSFORMPOSSOURCE_World}
  )
  Transform() {transformSourceType=TRANSFORMSOURCE_View}
  VectorParameter() {parameterName=LookAt, defaultValue=<rgba 14131 70226 4306 0>}
  WorldPosition()
 ]
}
