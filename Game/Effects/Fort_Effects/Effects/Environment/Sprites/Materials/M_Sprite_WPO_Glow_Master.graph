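# Translucent, unlit master material for glowing, bobbing environment sprites on instanced static meshes.
# The Sprite function drives world position offset, adding a sine bob and height offset varied per instance.
# A flare streak texture, randomly rotated per instance, shapes opacity under a tinted emissive color.
# Camera-distance masks fade distant sprites and hide near ones at randomized radii, with the near mask also shrinking sprite size.
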Material /Game/Effects/Fort_Effects/Effects/Environment/Sprites/Materials/M_Sprite_WPO_Glow_Master
graph {
 properties {
  bAutomaticallySetUsageInEditor=false
  bUsedWithInstancedStaticMeshes=true
  blendMode=BLEND_Translucent
  shadingModel=MSM_Unlit
  twoSided=true
  usedShadingModels=Unlit
 }
 nodes [
  n1 = MaterialFunctionCall!(
   Center Location=WorldPosition().XYZ
   0-1 UVs=TextureCoordinate() {coordinateIndex=1}
   XY scale=LinearInterpolate(B=ScalarParameter() {defaultValue=125, parameterName=Sprite Size}, Alpha=NamedRerouteUsage())
   OverridePivot
   "Normals (Optional)"
  ) {materialFunction=&Sprite}
  n2 = TransformPosition(None=ObjectPositionWS()) {transformSourceType=TRANSFORMPOSSOURCE_World, transformType=TRANSFORMPOSSOURCE_Instance}
  n3 = Add(A=Time(), B=PerInstanceRandom())
  n4 = ScalarParameter() {defaultValue=0.25, parameterName=Up and Down Speed}
  n5 = LinearInterpolate(
   A=Add(A=VectorParameter() {parameterName=Magnitue Low, defaultValue=<rgba 0 0 -1 0>}.RGB, B=@n2)
   B=Add(A=@n2, B=VectorParameter() {parameterName=Magnitude High, defaultValue=<rgba 0 0 10 0>}.RGB)
   Alpha=Sine(None=Multiply(A=@n3, B=@n4))
  )
  n6 = ObjectPositionWS()
  n7 = VectorParameter() {parameterName=Ground Offset Low, defaultValue=<rgba 0 0 22 0>}
  n8 = VectorParameter() {parameterName=Ground Offset High, defaultValue=<rgba 0 0 45 0>}
  n9 = PerInstanceRandom()
  Add(
   A=Multiply(A=Add(A=@n1.World position offset, B=@n5), B=ScalarParameter() {defaultValue=1, parameterName=Param_1})
   B=Add(
    A=TransformPosition(None=@n6) {transformSourceType=TRANSFORMPOSSOURCE_World, transformType=TRANSFORMPOSSOURCE_Instance}
    B=LinearInterpolate(A=@n7.RGB, B=@n8.RGB, Alpha=@n9)
   )
  )
  If("A > B"=Constant() {r=1}, "A == B"=Constant(), "A < B"=Constant()) {constB=0.75}
  MaterialFunctionCall(Emissive=VectorParameter() {parameterName=Sprite Color, defaultValue=<rgba 1 0.7791 0.5604 1>}.RGB) {materialFunction=&MF_FX_EmissiveAdjustments}
  n10 = WorldPosition()
  n11 = CameraPositionWS()
  n12 = ScalarParameter() {defaultValue=15000, parameterName=View Distance Fade}
  n13 = ScalarParameter() {parameterName=View Distance Mask Hardness}
  n14 = TextureCoordinate() {coordinateIndex=1}
  n15 = LinearInterpolate(Alpha=PerInstanceRandom()) {constB=360}
  n16 = Multiply(
   A=TextureSampleParameter2D(UVs=Rotator(Coordinate=@n14, Time=@n15)) {parameterName=Param, texture=/Game/Effects/Fort_Effects/Textures/Flares/T_Flare_Streaks}.G
   B=ScalarParameter() {defaultValue=0.05, parameterName=Param_3}
  )
  n17 = WorldPosition()
  n18 = CameraPositionWS()
  n19 = LinearInterpolate(
   A=ScalarParameter() {defaultValue=500, parameterName=Hide Low}
   B=ScalarParameter() {defaultValue=750, parameterName=Hide High}
   Alpha=PerInstanceRandom()
  )
  n20 = Power(
   Base=OneMinus(None=SphereMask(A=@n17.XYZ, B=@n18, Radius=@n19) {attenuationRadius=5500, hardnessPercent=0})
   Exp=ScalarParameter() {defaultValue=20, parameterName=Hide Falloff}
  )
  Multiply(
   A=Multiply(A=SphereMask(A=@n10.XYZ, B=@n11, Radius=@n12, Hardness=@n13), B=Multiply(A=@n16, B=@n20))
   B=ScalarParameter() {defaultValue=0.5, parameterName=Overall Opacity}
  )
  NamedRerouteDeclaration(None=@n20) {name=Distance, nodeColor=<rgba 0 1 0.0859 1>}
  ObjectPositionWS()
  ScalarParameter() {defaultValue=2500, parameterName=Hide Distance}
  TransformPosition() {transformSourceType=TRANSFORMPOSSOURCE_World, transformType=TRANSFORMPOSSOURCE_Instance}
 ]
}
